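; Spotlight fixture prop for an alien movie set, an adjustable light variant that poses its modified mesh and light in the construction script.
; After the parent construction runs, UpdateAddonRotations applies RodRotation as LightFixture yaw and FixtureRotation as fixture mesh roll.
; UpdateLightPosition then snaps the SpotLight to the mesh's first socket plus LightOffset, with pitch from negated FixtureRotation and yaw turned a quarter.
; BeginPlay, Tick and overlap events are empty.
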(event EventBeginPlay)

(event Collision|EventActorBeginOverlap
  (params (OtherActor :type Actor%20Object%20Reference)))

(event EventTick
  (params (DeltaSeconds :type Float%20%28single-precision%29)))

(fn UpdateAddonRotations
  (bind _sm_lightsource_spotlightfixture_a (Variables|Default|GetSM_LightSource_SpotLightFixture_A))
  (Transformation|SetRelativeRotation (Variables|AdjustableLight|GetLightFixture) (Math|Rotator|MakeRotator (.roll (Variables|/Script/Engine.SceneComponent|GetRelativeRotation (Variables|AdjustableLight|GetLightFixture))) (.pitch (Variables|/Script/Engine.SceneComponent|GetRelativeRotation (Variables|AdjustableLight|GetLightFixture))) (Variables|AdjustableLight|GetRodRotation)) false false)
  (if (Utilities|IsValid _sm_lightsource_spotlightfixture_a :!class /Script/Engine.KismetSystemLibrary)
    (Transformation|SetRelativeRotation _sm_lightsource_spotlightfixture_a (Math|Rotator|MakeRotator (Variables|AdjustableLight|GetFixtureRotation) (.pitch (Variables|/Script/Engine.SceneComponent|GetRelativeRotation _sm_lightsource_spotlightfixture_a)) (.yaw (Variables|/Script/Engine.SceneComponent|GetRelativeRotation _sm_lightsource_spotlightfixture_a))) false false)))

(fn UpdateLightPosition
  (bind (_location _rotation _scale) (.location (Transformation|GetSocketTransform (Variables|Default|GetSM_LightSource_SpotLightFixture_A) (Utilities|Array|Get(acopy) (Transformation|GetAllSocketNames (Variables|Default|GetSM_LightSource_SpotLightFixture_A))) "RTS_World")))
  (bind (_rotation_roll _rotation_pitch _rotation_yaw) (Math|Rotator|BreakRotator _rotation))
  (bind _sm_lightsource_spotlightfixture_a (Variables|Default|GetSM_LightSource_SpotLightFixture_A))
  (bind _returnvalue (Transformation|GetAllSocketNames _sm_lightsource_spotlightfixture_a))
  (if (Utilities|IsValid _sm_lightsource_spotlightfixture_a :!class /Script/Engine.KismetSystemLibrary)
    (if (> (Utilities|Array|Length (Actor|GetComponentsByClass self "/Script/Engine.LightComponent")) 0)
      (if (> (Utilities|Array|Length _returnvalue) 0)
        (Transformation|SetWorldTransform (Variables|Default|GetSpotLight) (Math|Transform|MakeTransform (+ (Variables|AdjustableLight|GetLightOffset) _location) (Math|Rotator|MakeRotator _rotation_roll (* (Variables|AdjustableLight|GetFixtureRotation) -1.0) (+ _rotation_yaw 9e+01)) _scale) false false)))))

(fn ConstructionScript
  (|CallParentFunction)
  ;[1 1 1] Additional logic for handling slightly modified mesh components
  (CallFunction|UpdateAddonRotations)
  (CallFunction|UpdateLightPosition))
