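; Umbrella light fixture Blueprint, a child of an adjustable light prop, that poses its meshes and spotlight at construction.
; After the parent's script, it applies RodRotation to LightFixture's yaw or, if that is invalid, FixtureRotation to the umbrella mesh's roll.
; It then snaps the SpotLight to the umbrella mesh's first socket plus LightOffset, angled by FixtureRotation, and assigns LightFixture's material.
; The event graph holds only empty stubs.
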(event EventBeginPlay)

(event Collision|EventActorBeginOverlap
  (params (OtherActor :type Actor%20Object%20Reference)))

(event EventTick
  (params (DeltaSeconds :type Float%20%28single-precision%29)))

(fn UpdateAddonRotations
  (bind _lightfixture (Variables|AdjustableLight|GetLightFixture))
  (if (Utilities|IsValid _lightfixture :!class /Script/Engine.KismetSystemLibrary)
    (Transformation|SetRelativeRotation _lightfixture (Math|Rotator|MakeRotator (.roll (Variables|/Script/Engine.SceneComponent|GetRelativeRotation _lightfixture)) (.pitch (Variables|/Script/Engine.SceneComponent|GetRelativeRotation _lightfixture)) (Variables|AdjustableLight|GetRodRotation)) false false)
    (Transformation|SetRelativeRotation (Variables|Default|GetSM_LightSource_UmbrellaLightFixture_A) (Math|Rotator|MakeRotator (Variables|AdjustableLight|GetFixtureRotation) (.pitch (Variables|/Script/Engine.SceneComponent|GetRelativeRotation (Variables|Default|GetSM_LightSource_UmbrellaLightFixture_A))) (.yaw (Variables|/Script/Engine.SceneComponent|GetRelativeRotation (Variables|Default|GetSM_LightSource_UmbrellaLightFixture_A)))) false false)))

(fn UpdateLightPosition
  (bind (_location _rotation _scale) (.location (Transformation|GetSocketTransform (Variables|Default|GetSM_LightSource_UmbrellaLightFixture_A) (Utilities|Array|Get(acopy) (Transformation|GetAllSocketNames (Variables|Default|GetSM_LightSource_UmbrellaLightFixture_A))) "RTS_World")))
  (bind (_rotation_roll _rotation_pitch _rotation_yaw) (Math|Rotator|BreakRotator _rotation))
  (bind _sm_lightsource_umbrellalightfixture_a (Variables|Default|GetSM_LightSource_UmbrellaLightFixture_A))
  (bind _returnvalue (Transformation|GetAllSocketNames _sm_lightsource_umbrellalightfixture_a))
  (bind _spotlight (Variables|Default|GetSpotLight))
  (if (Utilities|IsValid _sm_lightsource_umbrellalightfixture_a :!class /Script/Engine.KismetSystemLibrary)
    (if (> (Utilities|Array|Length (Actor|GetComponentsByClass self "/Script/Engine.LightComponent")) 0)
      (if (> (Utilities|Array|Length _returnvalue) 0)
        (if (Utilities|IsValid _spotlight :!class /Script/Engine.KismetSystemLibrary)
          (Transformation|SetWorldTransform _spotlight (Math|Transform|MakeTransform (+ (Variables|AdjustableLight|GetLightOffset) _location) (Math|Rotator|MakeRotator _rotation_roll (+ (* (Variables|AdjustableLight|GetFixtureRotation) -1.0) -3e+01) (+ _rotation_yaw 9e+01)) _scale) false false))))))

(fn ConstructionScript
  (|CallParentFunction)
  ;[1 1 1] Additional logic for handling slightly modified mesh components
  (Default|UpdateAddonRotations)
  (Default|UpdateLightPosition)
  (Rendering|Material|SetMaterial (Variables|AdjustableLight|GetLightFixture) 0 "/Hera/Props/LightSource/LightSource_TripodLights_A/Materials/MI_LightSorce_UmbrellaLightFixture_A.MI_LightSorce_UmbrellaLightFixture_A"))
